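# Parent material for the Dust Devil gadget prop, a stick and six colored drums that flash individually from per-drum start-time parameters.
# Rounded vertex-color red gives a drum ID, which switches between each drum's color and start time.
# Time since that start drives a short, power-curved emissive pulse, inserted between the shared character PreFX and PostFX functions.
# A DebugVC static switch offsets vertices by ID to check vertex-color assignment.
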Material /BRCosmetics/Animation/Props/Gadgets_DustDevil/Materials/M_DustDevilProp_Parent
graph {
 properties {
  bUseMaterialAttributes=true
  bUsedWithMeshDeformer=true
  bUsedWithMorphTargets=true
  bUsedWithNiagaraMeshParticles=true
  bUsedWithSkeletalMesh=true
  pixelDepthOffsetMode=PDOM_Legacy
  shadingModel=MSM_FromMaterialExpression
  usedShadingModels=Default Lit | Subsurface | Clear Coat | Subsurface Profile | Cloth
 }
 nodes [
  n1 = NamedRerouteDeclaration(None=Time()) {name=T, nodeColor=<rgba 1 0 0.5859 1>}
  n2 = ScalarParameter() {defaultValue=0.2, group=Flash Values, parameterName=FlashDuration}
  n3 = Power(Base=Frac(None=Divide(A=@n1, B=@n2)), Exp=ScalarParameter() {defaultValue=0.1, group=Flash Values, parameterName=CurvePower})
  n4 = OneMinus(None=Abs(None=ConstantBiasScale(None=@n3) {bias=-0.5, scale=2}))
  If(A=Frac(None=NamedRerouteUsage()), B=NamedRerouteUsage(), "A > B"=Constant(), "A == B"=@n4, "A < B"=@n4)
  n5 = MaterialFunctionCall() {materialFunction=&MF_CharacterParent_2023_PreFX}
  n6 = GetMaterialAttributes(None=@n5.Result) {attributeGetTypes [#1]}
  MaterialFunctionCall(
   Material=SetMaterialAttributes!(MaterialAttributes=@n6.MaterialAttributes, Emissive Color=NamedRerouteUsage(), World Position Offset=NamedRerouteUsage()) {attributeSetTypes [#1, #2]}
   M=@n5.M
   Pre Skinned Local Position=@n5.M
   Distance Blend=@n5.M
   Fresnel=@n5.M
  ) {materialFunction=&MF_CharacterParent_2023_PostFX}
  NamedRerouteDeclaration(None=@n6.Emissive Color) {name=EmissiveIn, nodeColor=<rgba 0 1 0.6953 1>}
  n7 = NamedRerouteUsage()
  n8 = Time()
  n9 = NamedRerouteUsage()
  n10 = Saturate(None=Divide(A=Subtract(A=@n8, B=@n9), B=ScalarParameter() {defaultValue=0.2, group=Flash Values, parameterName=FlashDuration}))
  n11 = ScalarParameter() {defaultValue=0.1, group=Flash Values, parameterName=CurvePower}
  n12 = Abs(None=ConstantBiasScale(None=Power(Base=@n10, Exp=@n11)) {bias=-0.5, scale=2})
  n13 = Multiply(A=@n7, B=Multiply(A=OneMinus(None=@n12), B=ScalarParameter() {defaultValue=10, group=Flash Values, parameterName=FlashIntensity}))
  NamedRerouteDeclaration(None=Multiply(A=Add(A=@n7, B=@n13), B=NamedRerouteUsage())) {name=EmissiveOut, nodeColor=<rgba 1 0.375 0 1>}
  NamedRerouteDeclaration(None=@n2) {name=Duration, nodeColor=<rgba 0 0.625 1 1>}
  n14 = Round(None=Multiply(A=VertexColor().R) {constB=6})
  n15 = Constant3Vector() {constant=<rgba 30 0 0 0>}
  NamedRerouteDeclaration(None=StaticSwitchParameter(True=Multiply(A=@n14, B=@n15), False=Constant()) {parameterName=DebugVC}) {name=WPO, nodeColor=<rgba 0.9922 0 1 1>}
  NamedRerouteDeclaration(None=@n14) {name=DrumID, nodeColor=<rgba 0.2656 0 1 1>}
  MaterialExpressionConstant3Vector_1 = Constant3Vector() {constant=<rgba 1 1 1 0>}
  MaterialExpressionVectorParameter_0 = VectorParameter() {group=Drums Color, parameterName=Drum1_Color, defaultValue=<rgba 0.0925 0.1541 0.9304 1>}
  MaterialExpressionVectorParameter_1 = VectorParameter() {group=Drums Color, parameterName=Drum2_Color, defaultValue=<rgba 0.2188 1 0.2859 1>}
  MaterialExpressionVectorParameter_2 = VectorParameter() {group=Drums Color, parameterName=Drum3_Color, defaultValue=<rgba 0.9304 0.2853 0.8615 1>}
  MaterialExpressionVectorParameter_3 = VectorParameter() {group=Drums Color, parameterName=Drum4_Color, defaultValue=<rgba 0.9304 0.526 0.1291 1>}
  MaterialExpressionVectorParameter_4 = VectorParameter() {group=Drums Color, parameterName=Drum5_Color, defaultValue=<rgba 0.6435 0.1276 0.0666 1>}
  MaterialExpressionVectorParameter_5 = VectorParameter() {group=Drums Color, parameterName=Drum6_Color, defaultValue=<rgba 0.9304 0.8866 0.1795 1>}
  n16 = Switch!(
   SwitchValue=NamedRerouteUsage()
   Default=@MaterialExpressionConstant3Vector_1
   Stick=@MaterialExpressionConstant3Vector_1
   Drum1=@MaterialExpressionVectorParameter_0.RGB
   Drum2=@MaterialExpressionVectorParameter_1.RGB
   Drum3=@MaterialExpressionVectorParameter_2.RGB
   Drum4=@MaterialExpressionVectorParameter_3.RGB
   Drum5=@MaterialExpressionVectorParameter_4.RGB
   Drum6=@MaterialExpressionVectorParameter_5.RGB
  ) {inputs [
    {
     inputName=Stick
     input {
      expression=&self:MaterialExpressionConstant3Vector_1
      inputName=None
      mask=1
      maskA=0
      maskB=1
      maskG=1
      maskR=1
      outputIndex=0
     }
    }
    {
     inputName=Drum1
     input {
      expression=&self:MaterialExpressionVectorParameter_0
      inputName=None
      mask=1
      maskA=0
      maskB=1
      maskG=1
      maskR=1
      outputIndex=0
     }
    }
    {
     inputName=Drum2
     input {
      expression=&self:MaterialExpressionVectorParameter_1
      inputName=None
      mask=1
      maskA=0
      maskB=1
      maskG=1
      maskR=1
      outputIndex=0
     }
    }
    {
     inputName=Drum3
     input {
      expression=&self:MaterialExpressionVectorParameter_2
      inputName=None
      mask=1
      maskA=0
      maskB=1
      maskG=1
      maskR=1
      outputIndex=0
     }
    }
    {
     inputName=Drum4
     input {
      expression=&self:MaterialExpressionVectorParameter_3
      inputName=None
      mask=1
      maskA=0
      maskB=1
      maskG=1
      maskR=1
      outputIndex=0
     }
    }
    {
     inputName=Drum5
     input {
      expression=&self:MaterialExpressionVectorParameter_4
      inputName=None
      mask=1
      maskA=0
      maskB=1
      maskG=1
      maskR=1
      outputIndex=0
     }
    }
    {
     inputName=Drum6
     input {
      expression=&self:MaterialExpressionVectorParameter_5
      inputName=None
      mask=1
      maskA=0
      maskB=1
      maskG=1
      maskR=1
      outputIndex=0
     }
    }
   ]}
  NamedRerouteDeclaration(None=@n16) {name=DrumColor, nodeColor=<rgba 0.0547 1 0 1>}
  MaterialExpressionConstant_2 = Constant() {r=-10000}
  MaterialExpressionScalarParameter_8 = ScalarParameter() {group=Drums TimeStart, parameterName=TimeStart_Yellow}
  MaterialExpressionScalarParameter_4 = ScalarParameter() {group=Drums TimeStart, parameterName=TimeStart_Green}
  MaterialExpressionScalarParameter_3 = ScalarParameter() {group=Drums TimeStart, parameterName=TimeStart_Blue}
  MaterialExpressionScalarParameter_5 = ScalarParameter() {group=Drums TimeStart, parameterName=TimeStart_Magenta}
  MaterialExpressionScalarParameter_6 = ScalarParameter() {group=Drums TimeStart, parameterName=TimeStart_Orange}
  MaterialExpressionScalarParameter_7 = ScalarParameter() {group=Drums TimeStart, parameterName=TimeStart_Red}
  n17 = Switch(
   SwitchValue=NamedRerouteUsage()
   Default=@MaterialExpressionConstant_2
   Stick=@MaterialExpressionConstant_2
   Drum1=@MaterialExpressionScalarParameter_8
   Drum2=@MaterialExpressionScalarParameter_4
   Drum3=@MaterialExpressionScalarParameter_3
   Drum4=@MaterialExpressionScalarParameter_5
   Drum5=@MaterialExpressionScalarParameter_6
   Drum6=@MaterialExpressionScalarParameter_7
  ) {inputs [
    {
     inputName=Stick
     input {
      expression=&self:MaterialExpressionConstant_2
      inputName=None
      mask=0
      maskA=0
      maskB=0
      maskG=0
      maskR=0
      outputIndex=0
     }
    }
    {
     inputName=Drum1
     input {
      expression=&self:MaterialExpressionScalarParameter_8
      inputName=None
      mask=0
      maskA=0
      maskB=0
      maskG=0
      maskR=0
      outputIndex=0
     }
    }
    {
     inputName=Drum2
     input {
      expression=&self:MaterialExpressionScalarParameter_4
      inputName=None
      mask=0
      maskA=0
      maskB=0
      maskG=0
      maskR=0
      outputIndex=0
     }
    }
    {
     inputName=Drum3
     input {
      expression=&self:MaterialExpressionScalarParameter_3
      inputName=None
      mask=0
      maskA=0
      maskB=0
      maskG=0
      maskR=0
      outputIndex=0
     }
    }
    {
     inputName=Drum4
     input {
      expression=&self:MaterialExpressionScalarParameter_5
      inputName=None
      mask=0
      maskA=0
      maskB=0
      maskG=0
      maskR=0
      outputIndex=0
     }
    }
    {
     inputName=Drum5
     input {
      expression=&self:MaterialExpressionScalarParameter_6
      inputName=None
      mask=0
      maskA=0
      maskB=0
      maskG=0
      maskR=0
      outputIndex=0
     }
    }
    {
     inputName=Drum6
     input {
      expression=&self:MaterialExpressionScalarParameter_7
      inputName=None
      mask=0
      maskA=0
      maskB=0
      maskG=0
      maskR=0
      outputIndex=0
     }
    }
   ]}
  NamedRerouteDeclaration(None=@n17) {name=DrumStart, nodeColor=<rgba 0 0.8594 1 1>}
 ]
}
